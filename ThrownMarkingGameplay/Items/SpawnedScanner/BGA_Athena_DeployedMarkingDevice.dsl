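; Deployed recon scanner, spawned from a throwable, that periodically pulses to mark hostile pawns and optionally chests for its owner's team.
; Server timers alternate anticipation cues with async targeting pings, each damaging the scanner toward a configured ping count.
; Marked players can reverse-mark the scanner and owner's squad, tuned by hotfixable ScalableFloats.
; Owner death, water or owner-limit gameplay events destroy it; OnReps handle charge and destruction cosmetics.
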(event EventBeginPlay
  ;[0 1 0.01759] BeginPlay
  ; Start marking timer and setup visuals
  (|CallParentFunction)
  (Networking|SwitchHasAuthority
    (:Authority
      (bind _asfort_player_pawn (Utilities|Casting|CastToFortPlayerPawn (Actor|GetOwner))
        (:then
          (bind _output_get (Variables|Default|SetPlayerPawnOwner _asfort_player_pawn))
          (if (not (Variables|/Script/FortniteGame.FortPawn|GetbIsDying _output_get))
            (Utilities|FlowControl|Sequence
              (:then_0
                ;; Handle if owner is eliminated, destroy this scanner.
                (bind _playerpawnowner (Variables|Default|GetPlayerPawnOwner)
                  (:then
                    (EventDispatchers|BindEventtoOnDied _playerpawnowner (EventDispatchers|CreateEvent :!event Custom|OwnerDied)))
                  (:else)))
              (:then_1
                (CallFunction|DeployScannerfromThrown)
                (EventDispatchers|BindEventtoOnDamaged (EventDispatchers|CreateEvent :!event Custom|EventOnDamaged))
                (Components|Activation|Deactivate (Variables|Default|GetNS_ThrownMarking_ChargeUp)))
              (:then_2
                (if (ScalableFloat|GetValueasBoolwithContext (Variables|Gameplay|Hotfixes|GetHFRow_OwnerLimitEnabled))
                  (CallFunction|SetMaxOwnerLimit)
                  (Ability|Async|WaitGameplayEventToActor 0 (Variables|Tags|GetEventTag_DeActivate_ActiveCount)
                    (:EventReceived
                      (CallFunction|DeactivateScanner))))))
            (else
              _)))
        (:CastFailed
          _))
      (CallFunction|DeactivateScanner))
    (:Remote
      (CallFunction|AdjustforTeamAffiliation))))

(event Custom|DeployScannerFromThrown
  (bind _hfrow_anticipationtimebeforescan (Variables|Gameplay|Hotfixes|GetHFRow_AnticipationTimeBeforeScan))
  (bind _returnvalue (ScalableFloat|GetValueatLevel _hfrow_anticipationtimebeforescan))
  ;[0.3828 0.5908 1 0.5] DeployScannerFromThrown
  ; Animation, VFX, and anything else to support the actor being deployed, before it begins scanning
  ;;[1 0.7826 0 0.5] RUN ANY DEPLOYMENT LOGIC NEEDED HERE AND THEN TRIGGER BEGIN SCANNING TIMER WHEN COMPLETED
  ;; - Animations, sfx, vfx, initialization, etc.
  (CallFunction|OnDeployed)
  ;;[0.3828 0.5908 1 0.5] After the deployment is complete, check if the scanner should immediately pulse or if it should play its anticipation VFX first
  (if (ScalableFloat|GetValueasBool (Variables|Gameplay|Hotfixes|GetHFRow_AnticipationFeedbackOnSpawnBool))
    (CallFunction|BeginScanAnticipationFeedback) ; run this immediately since there should be no delay after deployment
    (if (> _returnvalue 0)
      (bind _returnvalue_2 (Utilities|Time|SetTimerbyEvent (EventDispatchers|CreateEvent :!event Custom|PingRadiusForOwnerTeam) _returnvalue)) ; since we are only waiting for the anticipation feedback, \nimmediately scan after the feedback is done\nOther timers will be set based on the full interval duration
      (Variables|Default|SetPingRadiusTimerHandle _returnvalue_2)
      (else
        _))
    (else
      _))
  (CallFunction|PingRadiusforOwnerTeam) ; the first time the scan occurs should happen immediately, \nas soon as it is ready\nAfter the scan is triggered, a timer will begin for the next ping, if possible
)

(event Custom|BeginScanAnticipationFeedback
  ;[0.3709 0.003231 0.638 0.5] BeginScanAnticipationFeedback
  ; Execute gameplay cue slightly before the scan ping is sent to allow players to anticipate the action
  (Ability|ExecuteGameplayCueWithParamsReplicated (Variables|PlayerFeedback|GameplayCues|GetGCN_ScanChargeUpAnticipation) (Ability|GameplayCue|MakeGameplayCueParameters (ScalableFloat|GetValueatLevel (Variables|Gameplay|Hotfixes|GetHFRow_AnticipationTimeBeforeScan)) 0.0 (Class|AthenaFunctionLibrary|MakeSimpleFortEffectContext self (Actor|GetOwner))))
  (Networking|FlushNetDormancy)
  (Variables|FX|SetIsCharging true)
  (Variables|PlayerFeedback|GameplayCues|SetGCN_ScanChargeUpAnticipation (Variables|FX|GetGCNAfterInitialCharge)))

(event Custom|BeginNewScanningTimer
  (bind _hfrow_anticipationtimebeforescan (Variables|Gameplay|Hotfixes|GetHFRow_AnticipationTimeBeforeScan))
  (bind _returnvalue (ScalableFloat|GetValueatLevel _hfrow_anticipationtimebeforescan))
  (bind _hfrow_intervalbetweenmarks (Variables|Gameplay|Hotfixes|GetHFRow_IntervalBetweenMarks))
  (bind _returnvalue_2 (ScalableFloat|GetValueatLevel _hfrow_intervalbetweenmarks))
  (bind _returnvalue_3 (- _returnvalue_2 _returnvalue))
  ;[0 1 0.01759] BeginNewScanningTimer
  ; Delays when the next scan will occur - values are hotfixable
  (if (> _returnvalue_2 0)
    (bind _returnvalue_4 (Utilities|Time|SetTimerbyEvent (EventDispatchers|CreateEvent :!event Custom|PingRadiusForOwnerTeam) _returnvalue_2))
    (Variables|Default|SetPingRadiusTimerHandle _returnvalue_4)
    (if (> _returnvalue_3 0)
      (bind _returnvalue_5 (Utilities|Time|SetTimerbyEvent (EventDispatchers|CreateEvent :!event Custom|BeginScanAnticipationFeedback) _returnvalue_3)) ; Start the anticipation time X seconds before the scan will take place
      (Variables|Default|SetAnticipationFeedbackTimerHandle _returnvalue_5))
    (else
      ;;[1 0 0.8663 0.5] SCAN FAILED - requires a valid duration between pings
      (Fortnite|FortShippingLog self "PORTABLESCANNER: Interval radius timer evaluates less than zero, cannot set timer. DESTROYING NON-FUNCATIONAL OBJECT")
      (Actor|DestroyActor)
      (Networking|FlushNetDormancy)
      (Variables|Gameplay|SetbHasBeenDestroyed true))))

(event Custom|PingRadiusForOwnerTeam
  (bind _returnvalue (Fortnite|GetFortPlayerControllerfromActor (Actor|GetOwner)))
  (bind _hfrow_pingsbeforedestruction (Variables|Gameplay|Hotfixes|GetHFRow_PingsBeforeDestruction))
  (bind _returnvalue_2 (ScalableFloat|GetValueatLevel _hfrow_pingsbeforedestruction))
  (bind _returnvalue_3 (Actor|GetOwner))
  ;[0 0.09767 1 0.5] PingRadiusForOwnerTeam
  ; Check the area for hostile players and treasure chests, then give some damage to the scanner
  (Utilities|FlowControl|Sequence
    (:then_0
      (Utilities|IsValid _returnvalue_3
        (:Is%20Valid
          (CallFunction|StartPingEffect)
          (bind (_asynctaskproxy _targetinghandle) (Class|AsyncActionPerformTargeting|PerformTargetingAsyncAction 0 (Variables|Gameplay|Marking|GetTargeting%20Preset) true)
            (:Targeted
              ;;[0.15 0.1249 0.02073 0.5] Ping targeted pawns with filter
              (bind _markedactors (CallFunction|SetValidMarkedActors (TargetingSystem|TargetingResults|GetTargetingResultsActors (GameInstanceSubsystems|GetTargetingSubsystem) _targetinghandle)))
              (bind _modifieddata (CallFunction|ModifyFilterData (Variables|Gameplay|Marking|GetPawnFilters) (ScalableFloat|GetValueatLevel (Variables|Gameplay|Hotfixes|GetHFRow_MaxMarkingRadius)) (ScalableFloat|GetValueatLevel (Variables|Gameplay|Hotfixes|GetHFRow_EnemyMarkedDuration))))
              (IndicatedManagement|AddActorsToIndicatedListWithFilter _returnvalue _markedactors _modifieddata true true true _returnvalue_3 false)
              (CallFunction|ReplicatePingtoRelevantClients)
              (if (ScalableFloat|GetValueasBoolwithContext (Variables|Gameplay|Hotfixes|GetHFRow_ChestMarking))
                (bind _modifieddata_2 (CallFunction|ModifyFilterData (Variables|Gameplay|Marking|GetChestFilters) (ScalableFloat|GetValueatLevelwithContext (Variables|Gameplay|Hotfixes|GetChestMarkingRadius)) (ScalableFloat|GetValueatLevelwithContext (Variables|Gameplay|Hotfixes|GetHFRow_ChestMarkedDuration))))
                (IndicatedManagement|AddActorsInRadiusToIndicatedList _returnvalue _modifieddata_2 false true true self)
                (else
                  _))
              (Utilities|FlowControl|Delay (ScalableFloat|GetValueatLevel (Variables|Gameplay|Hotfixes|GetHFRow_DelayBeforeSelfDamageOnPing))) ; give a short delay so the scanner isn't \nimmediately destroyed on the final ping
              (if (> _returnvalue_2 0)
                (Game|Damage|ApplyDamage self (Math|Conversions|ToFloat(Integer) (Math|Float|Ceil (/ (Health|GetMaxHealth) _returnvalue_2))) 0 self "/Engine/EngineDamageTypes/DmgTypeBP_Environmental.DmgTypeBP_Environmental_C")
                (GameplayCue|ExecuteGameplayCueOnActor(Burst) self (Variables|PlayerFeedback|GameplayCues|GetGCN_PingScannerDamaged) (Ability|GameplayCue|MakeGameplayCueParameters :NormalizedMagnitude 0.0 :RawMagnitude (ScalableFloat|GetValueatLevel (Variables|Gameplay|Hotfixes|GetHFRow_MaxMarkingRadius)) :EffectContext (Class|AthenaFunctionLibrary|MakeSimpleFortEffectContext self (Actor|GetOwner)) :Location "0,0,0" :Normal "0,0,0" :SourceObject self :GameplayEffectLevel 1 :AbilityLevel 1 :bReplicateLocationWhenUsingMinimalRepProxy false))))))
        (:Is%20Not%20Valid
          (CallFunction|DeactivateScanner))))
    (:then_1
      (CallFunction|BeginNewScanningTimer) ; when the ping is complete, start the next ping timer, if possible
    )))

(event Custom|ReplicatePingToRelevantClients
  ;[0.3828 0.5908 1 0.5] ReplicatePingToRelevantClients
  ; Replicate the ping to clients within relevancy, does not need onRep for a burst effect
  (CallFunction|SendMarkFeedbacktoAffectedPawns))

(event OnEnterWaterEvent(FortWaterInteraction)
  (params (WaterBody :type Fort%20Water%20Body%20Actor%20Object%20Reference) (WaterInteractionComponent :type Fort%20Water%20Interaction%20Component%20Object%20Reference) (bIsFirstBody :type Boolean))
  ; Handle if scanner falls into water
  (Game|Components|ProjectileMovement|StopSimulating (Variables|Default|GetProjectileMovement) (Collision|MakeHitResult false false 0.0 0.0 "0, 0, 0" "0, 0, 0" "0,0,1" "0,0,1" 0 WaterBody))
  (CallFunction|DeactivateScanner))

(event Custom|StartPingEffect
  ;[0.4594 0 1] StartPingEffect
  ; Execute the large ping area of effect gameplay cue
  (Ability|ExecuteGameplayCueWithParamsReplicated (Variables|PlayerFeedback|GameplayCues|GetPulse%20Visual%20GC) (Ability|GameplayCue|MakeGameplayCueParameters :NormalizedMagnitude 0.0 :RawMagnitude (ScalableFloat|GetValueatLevel (Variables|Gameplay|Hotfixes|GetHFRow_MaxMarkingRadius)) :EffectContext (Class|AthenaFunctionLibrary|MakeSimpleFortEffectContext self (Actor|GetOwner)) :Location "0,0,0" :Normal "0,0,0" :SourceObject self :GameplayEffectLevel 1 :AbilityLevel 1 :bReplicateLocationWhenUsingMinimalRepProxy false))
  (Networking|FlushNetDormancy)
  (Variables|FX|SetIsCharging false))

(event EventOnDeathServer
  (params (Damage :type Float%20%28single-precision%29) (DamageTags :type Gameplay%20Tag%20Container%20Structure%20%28by%20ref%29) (Momentum :type Vector :default "0, 0, 0") (HitInfo :type Hit%20Result%20Structure%20%28by%20ref%29) (InstigatedBy :type Controller%20Object%20Reference) (DamageCauser :type Actor%20Object%20Reference) (EffectContext :type Gameplay%20Effect%20Context%20Handle%20Structure))
  (bind _returnvalue (Actor|GetOwner))
  ; Handle server death by sending the gameplay event back to the owner to correctly update the array of currently owned scanners.
  (|CallParentFunction Damage DamageTags Momentum HitInfo InstigatedBy DamageCauser EffectContext)
  (if (ScalableFloat|GetValueasBoolwithContext (Variables|Gameplay|Hotfixes|GetHFRow_OwnerLimitEnabled))
    (Utilities|IsValid _returnvalue
      (:Is%20Valid
        (Ability|SendGameplayEventtoActor _returnvalue (Variables|Tags|GetEventTag_RemoveOwnedScanner) (Utilities|Struct|MakeGameplayEventData "(TagName=\"\")" self))))))

(event Custom|EventOnDamaged
  (params (DamagedActor :type Actor%20Object%20Reference) (Damage :type Float%20%28single-precision%29) (InstigatedBy :type Controller%20Object%20Reference) (DamageCauser :type Actor%20Object%20Reference) (HitLocation :type Vector) (FHitComponent :type Primitive%20Component%20Object%20Reference) (BoneName :type Name) (Momentum :type Vector))
  ;[1 0.5548 0] EventOnDamaged
  ; Manually hide a destroyed actor if it is destroyed, is more responsive than standard replication to the client
  (if (<= (Health|GetHealthPercent) 0)
    (Networking|FlushNetDormancy)
    (Variables|Gameplay|SetbHasBeenDestroyed true)
    (Ability|SendGameplayEventtoActor (Actor|GetOwner) (Variables|Tags|GetEventTag_RemoveOwnedScanner) (Utilities|Struct|MakeGameplayEventData "(TagName=\"\")" self))))

(event Custom|SetMaxOwnerLimit
  ; When scanner spawns allow owner to control the personal limit if enabled.
  (bind _returnvalue (GameplayTags|HasMatchingGameplayTag(Message) (Actor|GetOwner) (Variables|Tags|GetTagScannerTracker)))
  (if (not _returnvalue)
    (GameplayEffects|ApplyGameplayEffectToTarget (Ability|GetFortAbilitySystemComponent self) (Variables|Gameplay|GetGE_ScannerTracker) (Ability|GetFortAbilitySystemComponent (Actor|GetOwner)) 0.0 (Class|AthenaFunctionLibrary|MakeSimpleFortEffectContext self self))
    (else
      _))
  (Ability|SendGameplayEventtoActor (Actor|GetOwner) (Variables|Tags|GetEventTag_AddOwnedScanner) (Utilities|Struct|MakeGameplayEventData "(TagName=\"\")" self)))

(event Custom|DeactivateScanner
  ; Called by event tag from tracker GA, to destroy itself and clean up as if it were destroyed.
  (Networking|FlushNetDormancy)
  (Variables|Gameplay|SetbHasBeenDestroyed true)
  (Gameplay|Die :DamageCauser self))

(event Custom|OwnerDied
  (params (DamagedActor :type Actor%20Object%20Reference) (Damage :type Float%20%28single-precision%29) (InstigatedBy :type Controller%20Object%20Reference) (DamageCauser :type Actor%20Object%20Reference) (HitLocation :type Vector) (FHitComponent :type Primitive%20Component%20Object%20Reference) (BoneName :type Name) (Momentum :type Vector))
  ; Handle the event of the owner being eliminated and destroy this scanner.
  (Utilities|FlowControl|Sequence
    (:then_0
      (bind _playerpawnowner (Variables|Default|GetPlayerPawnOwner)
        (:then
          (EventDispatchers|UnbindEventfromOnDied _playerpawnowner (delegate Custom|OwnerDied)))
        (:else)))
    (:then_1
      (if (not (Variables|Gameplay|GetbHasBeenDestroyed))
        (CallFunction|DeactivateScanner)))))

(fn AdjustForTeamAffiliation
  (bind _returnvalue (Fortnite|GetFirstLocalFortPlayerController true))
  (bind _returnvalue_2 (Actor|GetOwner))
  (bind _returnvalue_3 (Team|OnSameTeam _returnvalue _returnvalue_2))
  (Networking|CanExecuteCosmeticEvents
    (:True
      (Niagara|SetNiagaraVariable(Float) (Variables|Default|GetNS_ThrownMarking_ChargeUp) "EnnemyRecon" (Math|Float|SelectFloat 0.0 1.0 _returnvalue_3))
      (Variables|Default|SetbIsFriendly _returnvalue_3))))

(fn GetPingDelayByDistanceToMarkedActor
  (params (MarkedActorToCheck :type Actor%20Object%20Reference))
  (return (Math|Float|Lerp (Variables|PlayerFeedback|GetMinDelay) (Variables|PlayerFeedback|GetMaxDelay) (Math|Float|MapRangeClamped (Math|Vector|Distance(Vector) (Transformation|GetActorLocation) (Transformation|GetActorLocation MarkedActorToCheck)) (Variables|PlayerFeedback|GetDistanceMapMinRange) (ScalableFloat|GetValueatLevel (Variables|Gameplay|Hotfixes|GetHFRow_MaxMarkingRadius)) 0.0 1.0))))

(fn ModifyFilterData
  (params (FilterData :type Array%20of%20Indicated%20Actor%20Data%20With%20Filter%20Structures) (MarkingRadius :type Float%20%28double-precision%29) (MarkDuration :type Float%20%28double-precision%29))
  (bind _returnvalue (Actor|GetOwner))
  (Utilities|IsValid _returnvalue
    (:Is%20Valid
      (for _array_element FilterData
        (Utilities|Array|Add (LocalVariables|GethOutData) (Utilities|Struct|MakeIndicatedActorDatawithFilter (Utilities|Struct|BreakIndicatedActorDatawithFilter _array_element) (.IndicatedActorTagQuery (Utilities|Struct|BreakIndicatedActorDatawithFilter _array_element)) (.ObjectTypes (Utilities|Struct|BreakIndicatedActorDatawithFilter _array_element)) (.ActorClassFilter (Utilities|Struct|BreakIndicatedActorDatawithFilter _array_element)) (.WithAffiliation (Utilities|Struct|BreakIndicatedActorDatawithFilter _array_element)) (Utilities|Struct|MakeIndicatedActorData (Utilities|Struct|BreakIndicatedActorData (.IndicatedData (Utilities|Struct|BreakIndicatedActorDatawithFilter _array_element))) (Variables|GetMarkDuration) (.StepTime (Utilities|Struct|BreakIndicatedActorData (.IndicatedData (Utilities|Struct|BreakIndicatedActorDatawithFilter _array_element)))) (.IndicatorOffset (Utilities|Struct|BreakIndicatedActorData (.IndicatedData (Utilities|Struct|BreakIndicatedActorDatawithFilter _array_element)))) (.IndicatorDBNOOffset (Utilities|Struct|BreakIndicatedActorData (.IndicatedData (Utilities|Struct|BreakIndicatedActorDatawithFilter _array_element)))) (.ShareActorWith (Utilities|Struct|BreakIndicatedActorData (.IndicatedData (Utilities|Struct|BreakIndicatedActorDatawithFilter _array_element)))) (.ShareActorWithMask (Utilities|Struct|BreakIndicatedActorData (.IndicatedData (Utilities|Struct|BreakIndicatedActorDatawithFilter _array_element)))) true (Team|GetActorTeam _returnvalue) (.ScaleAndOpacityData (Utilities|Struct|BreakIndicatedActorData (.IndicatedData (Utilities|Struct|BreakIndicatedActorDatawithFilter _array_element)))) (.ParticleSystemData (Utilities|Struct|BreakIndicatedActorData (.IndicatedData (Utilities|Struct|BreakIndicatedActorDatawithFilter _array_element)))) (.Sound (Utilities|Struct|BreakIndicatedActorData (.IndicatedData (Utilities|Struct|BreakIndicatedActorDatawithFilter _array_element)))) (.StateImageOverride (Utilities|Struct|BreakIndicatedActorData (.IndicatedData (Utilities|Struct|BreakIndicatedActorDatawithFilter _array_element)))) (.StateIndexOverride (Utilities|Struct|BreakIndicatedActorData (.IndicatedData (Utilities|Struct|BreakIndicatedActorDatawithFilter _array_element)))) (.bClampToScreen (Utilities|Struct|BreakIndicatedActorData (.IndicatedData (Utilities|Struct|BreakIndicatedActorDatawithFilter _array_element)))) (.bHideStateImage (Utilities|Struct|BreakIndicatedActorData (.IndicatedData (Utilities|Struct|BreakIndicatedActorDatawithFilter _array_element)))) (.bOnlyDisplayWhenStencilingIsDisabled (Utilities|Struct|BreakIndicatedActorData (.IndicatedData (Utilities|Struct|BreakIndicatedActorDatawithFilter _array_element)))) false (.OnActorAdded (Utilities|Struct|BreakIndicatedActorData (.IndicatedData (Utilities|Struct|BreakIndicatedActorDatawithFilter _array_element)))) (.OnActorRemoved (Utilities|Struct|BreakIndicatedActorData (.IndicatedData (Utilities|Struct|BreakIndicatedActorDatawithFilter _array_element)))) (.EventTags (Utilities|Struct|BreakIndicatedActorData (.IndicatedData (Utilities|Struct|BreakIndicatedActorDatawithFilter _array_element))))) (Utilities|Struct|MakeStenciledActorData (Utilities|Struct|BreakStenciledActorData (.StenciledData (Utilities|Struct|BreakIndicatedActorDatawithFilter _array_element))) (Variables|GetMarkDuration) (.StepTime (Utilities|Struct|BreakStenciledActorData (.StenciledData (Utilities|Struct|BreakIndicatedActorDatawithFilter _array_element)))) (.ShareActorWithMask (Utilities|Struct|BreakStenciledActorData (.StenciledData (Utilities|Struct|BreakIndicatedActorDatawithFilter _array_element)))) true (.DisplayTeamOverride (Utilities|Struct|BreakStenciledActorData (.StenciledData (Utilities|Struct|BreakIndicatedActorDatawithFilter _array_element)))) (.ParticleSystemData (Utilities|Struct|BreakStenciledActorData (.StenciledData (Utilities|Struct|BreakIndicatedActorDatawithFilter _array_element)))) (.Sound (Utilities|Struct|BreakStenciledActorData (.StenciledData (Utilities|Struct|BreakIndicatedActorDatawithFilter _array_element)))) (.FriendlyStencilName (Utilities|Struct|BreakStenciledActorData (.StenciledData (Utilities|Struct|BreakIndicatedActorDatawithFilter _array_element)))) (.EnemyStencilName (Utilities|Struct|BreakStenciledActorData (.StenciledData (Utilities|Struct|BreakIndicatedActorDatawithFilter _array_element))))) MarkingRadius (.OverlapHalfHeight (Utilities|Struct|BreakIndicatedActorDatawithFilter _array_element)))))
      (return (LocalVariables|GethOutData)))))

(fn OnDeployed
  ; Play Deploy Animation Montage
  (Animation|Utilities|PlayAnimMontage 0 "/ThrownMarkingGameplay/Assets/Animation/ReconScanner_Deploy_Gadget_M.ReconScanner_Deploy_Gadget_M")
  ; Execute deployed Cue (if desired).
  (GameplayCue|ExecuteGameplayCueOnActor(Burst) self (Variables|PlayerFeedback|GameplayCues|GetGCN_PingScannerDeployed) (Ability|GameplayCue|MakeGameplayCueParameters :NormalizedMagnitude 0.0 :RawMagnitude 0.0 :Location (Transformation|GetActorLocation self) :Normal "0,0,0" :GameplayEffectLevel 1 :AbilityLevel 1 :bReplicateLocationWhenUsingMinimalRepProxy false))
  (Networking|CanExecuteCosmeticEvents
    (:True
      (Development|PrintString "Deployed  Valid"))
    (:False
      (Development|PrintString "Deployed Not Valid"))))

(fn OnRep_bHasBeenDestroyed
  (bind _self self)
  (if (Variables|Gameplay|GetbHasBeenDestroyed)
    (Networking|CanExecuteCosmeticEvents
      (:True
        (Utilities|FlowControl|Sequence
          (:then_0
            (bind _staticmesh (Variables|Default|GetStaticMesh)
              (:then
                (Rendering|SetVisibility _staticmesh false true))
              (:else)))
          (:then_1
            (bind _scannerloop (Variables|Default|GetScannerLoop)
              (:then
                (Audio|Components|Audio|Stop _scannerloop)
                (SoundIndicator|StopSoundIndicatorLooping (Variables|Default|GetFortSoundIndicator_ScannerLoop)))
              (:else)))
          (:then_2
            (GameplayCue|ExecuteGameplayCue%28Burst%29Local self (Variables|PlayerFeedback|GameplayCues|GetGCN_PingScannerDestroyed) (Ability|GameplayCue|MakeGameplayCueParameters 0.0 (Math|Float|SelectFloat 0.0 1.0 (Variables|Default|GetbIsFriendly)))) ; Execute the correct VFX based on affiliation
          )))
      (:False
        (Utilities|Time|ClearandInvalidateTimerbyHandle (Variables|Default|GetPingRadiusTimerHandle))
        (Utilities|Time|ClearandInvalidateTimerbyHandle (Variables|Default|GetAnticipationFeedbackTimerHandle))
        ; Remove the scanner from the indicated actor list. Removes the icon when the actor is destroyed instead of letting it staying around even after the actor is destroyed.
        (if (ScalableFloat|GetValueasBoolwithContext (Variables|Gameplay|Hotfixes|GetHFRow_ReverseMarkingIconEnabled))
          (for _array_element (Variables|Gameplay|Marking|GetMarkedActors)
            (Utilities|IsValid _array_element
              (:Is%20Valid
                (Utilities|IsValid (Fortnite|GetControllerfromActor _array_element)
                  (:Is%20Valid
                    (Utilities|IsValid _self
                      (:Is%20Valid
                        (IndicatedManagement|RemoveActorFromIndicatedList (Fortnite|GetControllerfromActor _array_element) _self)))))))))))))

(fn OnRep_IsCharging
  (bind _ischarging (Variables|FX|GetIsCharging))
  (bind _scannerloop (Variables|Default|GetScannerLoop))
  (bind _ns_thrownmarking_chargeup (Variables|Default|GetNS_ThrownMarking_ChargeUp))
  (Networking|CanExecuteCosmeticEvents
    (:True
      (if _ischarging
        (Niagara|SetNiagaraVariable(Float) _ns_thrownmarking_chargeup "ChargeUpDuration" (ScalableFloat|GetValueatLevel (Variables|Gameplay|Hotfixes|GetHFRow_AnticipationTimeBeforeScan)))
        (Audio|Components|Audio|FadeOut _scannerloop 0.2 0.0 "Linear" :!class /Script/Engine.AudioComponent)
        (SoundIndicator|StopSoundIndicatorLooping (Variables|Default|GetFortSoundIndicator_ScannerLoop))
        (else
          (Audio|Components|Audio|Play _scannerloop 2.0)
          (Utilities|Time|SetTimerbyEvent (EventDispatchers|CreateEvent :!event SetSoundIndicatorVisibility) 2.45)))
      (Niagara|SetNiagaraVariable(Bool) _ns_thrownmarking_chargeup "IsCharging" _ischarging))))

(fn ReverseMarkingScannerForDetectedEnemies
  (params (EnemyPlayerController :type Fort%20Player%20Controller%20Object%20Reference))
  (bind (_groupidentifier _duration _steptime _shareactorwith _shareactorwithmask _bvisibletoowner _displayteamoverride _particlesystemdata _sound _friendlystencilname _enemystencilname) (Utilities|Struct|BreakStenciledActorData (Variables|Gameplay|Marking|GetScannerCounterGameplayStenciledData)))
  (bind _returnvalue (Fortnite|Athena|GameState|GetGameStateAthena))
  (bind _self self)
  (bind _array (Utilities|Array|MakeArray _self))
  (bind _enemyplayercontroller (Variables|GetEnemyPlayerController))
  (Utilities|FlowControl|Sequence
    (:then_0
      (if (ScalableFloat|GetValueasBoolwithContext (Variables|Gameplay|Hotfixes|GetHFRow_ReverseMarkingStencilEnabled))
        (IndicatedManagement|AddActorsToStenciledList _enemyplayercontroller _array (Utilities|Struct|MakeStenciledActorData _groupidentifier (ScalableFloat|GetValueatLevel (Variables|Gameplay|Hotfixes|GetHFRow_ReverseMarkingStencilTime)) _steptime _shareactorwithmask true _displayteamoverride _particlesystemdata _sound _friendlystencilname _enemystencilname) false false true)))
    (:then_1
      (if (ScalableFloat|GetValueasBoolwithContext (Variables|Gameplay|Hotfixes|GetHFRow_ReverseMarkingIconEnabled))
        (IndicatedManagement|AddActorsToIndicatedList _enemyplayercontroller _array (Utilities|Struct|MakeIndicatedActorData "scanner" (ScalableFloat|GetValueatLevelwithContext (Variables|Gameplay|Hotfixes|GetHFRow_ReverseMarkingIconTime)) 0.0 "0.000000,0.000000,150.000000" "0.000000,0.000000,150.000000" "None" "SquadOnTeam" true 255 "(SmallSizeDistance=7500.000000,LargestSizeDistance=10.000000,SmallScale=0.500000,LargestScale=0.800000,FarAwayScale=0.350000,FarAwayOpacity=1.000000,MaxScaleAndFadePercent=0.100000)" "(ParticleSystem=None,ParticleSystemActorParamName=\"\",ParticleSystemOffset=(X=0.000000,Y=0.000000,Z=0.000000),ParticleSystemDBNOOffset=(X=0.000000,Y=0.000000,Z=0.000000),ParticleSystemVectorParamName=\"\")" 0 "Enemy") true false false true)))
    (:then_2
      (if (ScalableFloat|GetValueasBoolwithContext (Variables|Gameplay|Hotfixes|GetHFRow_ReverseMarkingMarkerEnabled))
        (MarkerComponent|MarkActoronServer (Actor|GetComponentbyClass (Variables|GetEnemyPlayerController) "/Script/FortniteGame.AthenaMarkerComponent") self true true)))
    (:then_3
      ; Handle Reverse marking the owner and/or the owner's team if wanted.
      (if (ScalableFloat|GetValueasBoolwithContext (Variables|Gameplay|Hotfixes|GetHFRow_ReverseMarkOwner))
        (if (ScalableFloat|GetValueasBoolwithContext (Variables|Gameplay|Hotfixes|GetHFRow_ReverseMarkOwnerTeam))
          (Utilities|IsValid _returnvalue
            (:Is%20Valid
              (bind _asfort_player_pawn (Utilities|Casting|CastToFortPlayerPawn (Actor|GetOwner)))
              (bind _returnvalue_2 (Fortnite|Utilities|AttemptGetPlayerStateforPawn _asfort_player_pawn))
              (bind _asfort_player_state_athena (Utilities|Casting|CastToFortPlayerStateAthena _returnvalue_2))
              (bind _returnvalue_3 (Class|FortGameStateAthena|GetSquadMemberPlayerStates _returnvalue _asfort_player_state_athena true))
              (for _array_element _returnvalue_3
                (Utilities|IsValid _array_element
                  (:Is%20Valid
                    (if (not (Variables|/Script/Engine.PlayerState|GetbIsSpectator _array_element))
                      (Utilities|IsValid (PlayerState|GetPawn _array_element)
                        (:Is%20Valid
                          (Utilities|Array|AddUnique (LocalVariables|GetOwnerTeamForReverseMarking) (PlayerState|GetPawn _array_element))))))))))
          (else
            (bind _asfort_player_pawn_2 (Utilities|Casting|CastToFortPlayerPawn (Actor|GetOwner)))
            (Utilities|Array|AddUnique (LocalVariables|GetOwnerTeamForReverseMarking) _asfort_player_pawn_2)))
        (bind _modifieddata (CallFunction|ModifyFilterData (Variables|Gameplay|Marking|GetPawnFilters) (ScalableFloat|GetValueatLevel (Variables|Gameplay|Hotfixes|GetHFRow_MaxMarkingRadius)) (ScalableFloat|GetValueatLevelwithContext (Variables|Gameplay|Hotfixes|GetHFRow_OwnerTeamMarkedDuration))))
        (IndicatedManagement|AddActorsToIndicatedListWithFilter (Variables|GetEnemyPlayerController) (LocalVariables|GetOwnerTeamForReverseMarking) _modifieddata true false true (Variables|GetEnemyPlayerController) false)))))

(fn SendMarkFeedbackToAffectedPawns
  (bind _returnvalue (Fortnite|Athena|GameState|GetGameStateAthena))
  (for _array_element (Variables|Gameplay|Marking|GetMarkedActors)
    (Networking|CanExecuteCosmeticEvents
      (:True
        ;[0.5 0.5 0.5] TODO: This should be handled better, and should support more than two possible marked targets
        (Utilities|FlowControl|Sequence
          (:then_0
            (Utilities|IsValid _returnvalue
              (:Is%20Valid
                (bind _asfort_player_pawn (Utilities|Casting|CastToFortPlayerPawn (Actor|GetOwner)))
                (bind _returnvalue_2 (Fortnite|Utilities|AttemptGetPlayerStateforPawn _asfort_player_pawn))
                (bind _asfort_player_state_athena (Utilities|Casting|CastToFortPlayerStateAthena _returnvalue_2))
                (bind _returnvalue_3 (Class|FortGameStateAthena|GetSquadMemberPlayerStates _returnvalue _asfort_player_state_athena true))
                (bind _array_element_2 (Utilities|Array|ForEachLoopwithBreak (Fortnite|GetLocalFortPlayerControllers true true))
                  (:LoopBody
                    (bind _asfort_player_state_athena_2 (Utilities|Casting|CastToFortPlayerStateAthena (Variables|/Script/Engine.Controller|GetPlayerState _array_element_2)))
                    (if (Utilities|Array|ContainsItem _returnvalue_3 _asfort_player_state_athena_2)
                      (LocalVariables|Seth_ShouldPlayToLocalClient true)))))))
          (:then_1
            (if (LocalVariables|Geth_ShouldPlayToLocalClient)
              (GameplayCue|ExecuteGameplayCue%28Burst%29Local _array_element (Variables|PlayerFeedback|GameplayCues|GetGCN_MarkingByOwner) (Ability|GameplayCue|MakeGameplayCueParameters :NormalizedMagnitude 0.0 :RawMagnitude (CallFunction|GetPingDelaybyDistancetoMarkedActor _array_element) :Location (Transformation|GetActorLocation _array_element) :Normal "0,0,0" :GameplayEffectLevel 1 :AbilityLevel 1 :bReplicateLocationWhenUsingMinimalRepProxy false))))))
      (:False
        ;[0 0.2392 1 0.5] Apply GE to all Marked players to replicate the mark effect to their local client
        (bind _asfort_player_pawn_athena (Utilities|Casting|CastToFortPlayerPawnAthena _array_element))
        (Creative|GameplayEffects|SimpleApplyGameplayEffectToActor (Variables|PlayerFeedback|GetGE_MarkedByOpponent) _asfort_player_pawn_athena)
        ;[0.3223 0 1] For players who are MARKED,  mark the Scanner -> encourages counterplay
        (if (ScalableFloat|GetValueasBool (Variables|Gameplay|Hotfixes|GetHFRow_ReverseMarkingEnabled))
          (CallFunction|ReverseMarkingScannerforDetectedEnemies (Fortnite|GetFortPlayerControllerfromActor _asfort_player_pawn_athena)))))))

(fn SetSoundIndicatorVisibility
  (SoundIndicator|StartSoundIndicatorLooping (Variables|Default|GetFortSoundIndicator_ScannerLoop) (Transformation|GetActorLocation) 0 1.0 0))

(fn SetValidMarkedActors
  (params (TargetActorsToCheck :type Array%20of%20Actor%20Object%20References))
  (bind _returnvalue (Actor|GetOwner))
  (for _array_element TargetActorsToCheck
    ;[0.3313 0.5313 1] Verify the targeted actors, make sure they are valid for marking (not on the same team, etc.)
    (Utilities|IsValid _returnvalue
      (:Is%20Valid
        (switch Utilities|FlowControl|Switch|SwitchonEFortTeamAffiliation (Team|GetTeamAffiliation _returnvalue _array_element)
          (:Neutral
            (bind _returnvalue_2 (GameplayTags|HasMatchingGameplayTag(Message) _array_element "(TagName=\"AISpawnerData.Type.VividRazor.Minions\")")) ; Mark Neutral AI if they are minions around the vaults/convoy but not yet aggressive.
            (if _returnvalue_2
              _))
          (:Hostile
            _))
        (Utilities|Array|AddUnique (LocalVariables|GethValidTargetedActorsToMark) _array_element))))
  ;[0.9302 0.5156 1] Return and store only valid actors for marking
  (Networking|FlushNetDormancy)
  (bind _output_get (Variables|Gameplay|Marking|SetMarkedActors (LocalVariables|GethValidTargetedActorsToMark)))
  (return _output_get))

(fn ConstructionScript
  (|CallParentFunction))
